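# Niagara system for a multi-bubble soup effect on the CrewCut buffet food prop, judging by its name and path.
# It holds four emitters, named Purple, Red, Yellow and Empty001, with the color names suggesting per-color bubble variants.
# No renderers are listed for any emitter, SystemState is the only recorded module, and the system exposes no user variables.
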NiagaraSystem /Hermes/Props/CrewCut/CrewCut_BuffetFood_A/FX/NS_CrewCut_BubblingSoup_MultiBubble
dependencies {usedDataInterfaces [], usedDynamicInputs [], usedModules [&SystemState]}
emitters {
 Empty001 {rendererClasses [], renderers []}
 Purple {rendererClasses [], renderers []}
 Red {rendererClasses [], renderers []}
 Yellow {rendererClasses [], renderers []}
}
summary {
 systemName=NS_CrewCut_BubblingSoup_MultiBubble
 emitters [
  {emitterName=Purple, rendererClasses []}
  {emitterName=Empty001, rendererClasses []}
  {emitterName=Yellow, rendererClasses []}
  {emitterName=Red, rendererClasses []}
 ]
 userVariables []
}
